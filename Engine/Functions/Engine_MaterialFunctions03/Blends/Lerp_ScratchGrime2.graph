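# Lerp_ScratchGrime2 is a library material function in Misc/Blends for building material layers with scratches and grime.
# It first lerps Base toward Base plus ScratchValue by the scalar ScratchMask, so scratches are additive.
# That result is then lerped toward itself multiplied by GrimeValue using the scalar GrimeMask, so grime tints or darkens on top of the scratches.
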MaterialFunction /Engine/Functions/Engine_MaterialFunctions03/Blends/Lerp_ScratchGrime2
graph {
 properties {bExposeToLibrary=true, description=Use to creat Material layers that have scratches and grime, libraryCategoriesText [Misc,Blends]}
 nodes [
  n1 = FunctionInput() {inputName=Base}
  n2 = LinearInterpolate(
   A=@n1
   B=Add(A=@n1, B=FunctionInput() {inputName=ScratchValue, sortPriority=1})
   Alpha=FunctionInput() {inputName=ScratchMask, inputType=FunctionInput_Scalar, sortPriority=3}
  )
  n3 = FunctionInput() {inputName=GrimeValue, sortPriority=2}
  FunctionOutput(None=LinearInterpolate(A=@n2, B=Multiply(A=@n2, B=@n3), Alpha=FunctionInput() {inputName=GrimeMask, inputType=FunctionInput_Scalar, sortPriority=4}))
 ]
}
