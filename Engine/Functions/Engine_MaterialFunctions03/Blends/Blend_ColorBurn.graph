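# Material function that applies a Color Burn blend between a Base and a Blend input, exposed to the material library under Misc and Blends.
# It computes one minus the inverted Base divided by Blend, so darker Blend values intensify the result and white leaves Base unchanged.
# Both inputs preview with default engine blend textures, and the output is not clamped.
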MaterialFunction /Engine/Functions/Engine_MaterialFunctions03/Blends/Blend_ColorBurn
graph {
 properties {
  bExposeToLibrary=true
  description="The darker the Blend texture, the more color in the final result.  If the Blend is white, there will be no change."
  libraryCategoriesText [Misc,Blends]
 }
 nodes [
  n1 = OneMinus(None=FunctionInput(Preview=TextureSample() {texture=/Engine/EngineMaterials/BlendFunc_DefBase}.RGB) {description=Base texture input, inputName=Base})
  n2 = FunctionInput(Preview=TextureSample() {texture=/Engine/EngineMaterials/BlendFunc_DefBlend}.RGB) {description=Blend texture input, inputName=Blend}
  FunctionOutput(None=OneMinus(None=Divide(A=@n1, B=@n2))) {description="The darker the Blend texture, the more color in the final result.  If the Blend is white, there will be no change."}
 ]
}
