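# Blend_ColorDodge is a library material function under Misc and Blends that applies a Color Dodge blend to two inputs.
# It divides the Base input by one minus the Blend input and outputs the result, with no clamp or guard against division by zero.
# Default engine BlendFunc textures feed the preview values of both inputs.
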MaterialFunction /Engine/Functions/Engine_MaterialFunctions03/Blends/Blend_ColorDodge
graph {
 properties {bExposeToLibrary=true, description=Divides the Base texture by the inverted Blend texture., libraryCategoriesText [Misc,Blends]}
 nodes [
  n1 = TextureSample() {texture=/Engine/EngineMaterials/BlendFunc_DefBase}
  n2 = FunctionInput(Preview=TextureSample() {texture=/Engine/EngineMaterials/BlendFunc_DefBlend}.RGB) {description=Blend texture input, inputName=Blend}
  FunctionOutput(None=Divide(A=FunctionInput(Preview=@n1.RGB) {description=Base texture input, inputName=Base}, B=OneMinus(None=@n2))) {description=Divides the Base texture by the inverted Blend texture.}
 ]
}
