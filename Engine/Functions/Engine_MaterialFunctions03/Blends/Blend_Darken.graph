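# Material function that implements a Darken blend, returning the darker of two values for each pixel, exposed to the library under Misc and Blends.
# Base and Blend inputs, each with a default texture sample for preview, feed a Custom HLSL node.
# That node takes the component-wise min of the two and passes it straight to the single output.
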MaterialFunction /Engine/Functions/Engine_MaterialFunctions03/Blends/Blend_Darken
graph {
 properties {bExposeToLibrary=true, description=Returns the darker of the two values for each pixel., libraryCategoriesText [Misc,Blends]}
 nodes [
  MaterialExpressionFunctionInput_6 = FunctionInput(Preview=TextureSample() {texture=/Engine/EngineMaterials/BlendFunc_DefBase}.RGB) {description=Base texture input, inputName=Base}
  MaterialExpressionFunctionInput_8 = FunctionInput(Preview=TextureSample() {texture=/Engine/EngineMaterials/BlendFunc_DefBlend}.RGB) {description=Blend texture input, inputName=Blend}
  n1 = Custom!(In1=@MaterialExpressionFunctionInput_6, In2=@MaterialExpressionFunctionInput_8) {
   code="min(In1, In2)\n"
   inputs [
    {inputName=In1, input {expression=&self:MaterialExpressionFunctionInput_6}}
    {inputName=In2, input {expression=&self:MaterialExpressionFunctionInput_8}}
   ]
  }
  FunctionOutput(None=@n1) {description=Returns the darker of the two values for each pixel.}
 ]
}
