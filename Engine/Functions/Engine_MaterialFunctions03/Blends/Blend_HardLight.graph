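# Material function in the Misc/Blends library that applies a Hard Light blend of a Blend input onto a Base input, described as a harsher Soft Light.
# It computes a screen term from Base and remapped Blend, and a multiply term from Base and doubled Blend.
# Per-channel If nodes pick screen where Blend exceeds 0.5 and multiply below it.
# The channels are appended back into RGB and output through Abs.
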MaterialFunction /Engine/Functions/Engine_MaterialFunctions03/Blends/Blend_HardLight
graph {
 properties {bExposeToLibrary=true, description=Harsher version of Soft Light blend, libraryCategoriesText [Misc,Blends]}
 nodes [
  n1 = FunctionInput(Preview=TextureSample() {texture=/Engine/EngineMaterials/BlendFunc_DefBlend}.RGB) {desc=Blend, description=Blend texture input, inputName=Blend}
  n2 = Constant() {r=0.5}
  n3 = FunctionInput(Preview=TextureSample() {texture=/Engine/EngineMaterials/BlendFunc_DefBase}.RGB) {desc=Base, description=Base texture input, inputName=Base}
  n4 = Multiply(A=Add(A=@n1, B=Constant() {r=-0.5}), B=Constant() {r=2})
  n5 = OneMinus(None=Multiply(A=OneMinus(None=@n3), B=OneMinus(None=@n4)))
  n6 = Multiply(A=@n3, B=Multiply(A=@n1, B=Constant() {r=2}))
  n7 = AppendVector(
   A=If(A=ComponentMask(None=@n1) {r=true}, B=@n2, "A > B"=ComponentMask(None=@n5) {r=true}, "A < B"=ComponentMask(None=@n6) {r=true})
   B=If(A=ComponentMask(None=@n1) {g=true}, B=@n2, "A > B"=ComponentMask(None=@n5) {g=true}, "A < B"=ComponentMask(None=@n6) {g=true})
  )
  n8 = If(A=ComponentMask(None=@n1) {b=true}, B=@n2, "A > B"=ComponentMask(None=@n5) {b=true}, "A < B"=ComponentMask(None=@n6) {b=true})
  FunctionOutput(None=Abs(None=AppendVector(A=@n7, B=@n8))) {description=Harsher version of Soft Light blend}
 ]
}
