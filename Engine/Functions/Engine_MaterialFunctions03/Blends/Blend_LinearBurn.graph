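# Material function that performs a Linear Burn blend of a Base input and a Blend input, exposed to the material library under Misc and Blends.
# It adds Base and Blend, then adds a constant of negative one, outputting the result without clamping.
# Both inputs use default engine blend textures as previews.
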MaterialFunction /Engine/Functions/Engine_MaterialFunctions03/Blends/Blend_LinearBurn
graph {
 properties {bExposeToLibrary=true, description=Adds the two textures then subtracts one from the result, libraryCategoriesText [Misc,Blends]}
 nodes [
  n1 = FunctionInput(Preview=TextureSample() {texture=/Engine/EngineMaterials/BlendFunc_DefBase}.RGB) {description=Base texture input, inputName=Base}
  n2 = FunctionInput(Preview=TextureSample() {texture=/Engine/EngineMaterials/BlendFunc_DefBlend}.RGB) {description=Blend texture input, inputName=Blend}
  FunctionOutput(None=Add(A=Add(A=@n1, B=@n2), B=Constant() {r=-1})) {description=Adds the two textures then subtracts one from the result}
 ]
}
