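# Material function exposed under Misc/Blends that computes a Pin Light blend of a Base and a Blend input.
# Per RGB channel, Blend above 0.5 selects max(Base, 2*(Blend-0.5)), while Blend below 0.5 selects min(Base, 2*Blend).
# Custom HLSL nodes supply max and min, and three If results are appended into the output.
# The descriptions mention Overlay instead, and a stray Multiply/OneMinus chain never reaches the output.
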MaterialFunction /Engine/Functions/Engine_MaterialFunctions03/Blends/Blend_PinLight
graph {
 properties {
  bExposeToLibrary=true
  description="Softer version of Overlay, solid white or black as blend will not result in solid color."
  libraryCategoriesText [Misc,Blends]
 }
 nodes [
  n1 = FunctionInput(Preview=TextureSample() {texture=/Engine/EngineMaterials/BlendFunc_DefBlend}.RGB) {description=Blend texture input, inputName=Blend}
  n2 = Constant() {r=0.5}
  MaterialExpressionFunctionInput_0 = FunctionInput(Preview=TextureSample() {texture=/Engine/EngineMaterials/BlendFunc_DefBase}.RGB) {description=Base texture input, inputName=Base}
  MaterialExpressionMultiply_6 = Multiply(A=Add(A=@n1, B=Constant() {r=-0.5}), B=Constant() {r=2})
  n3 = Custom!(base=@MaterialExpressionFunctionInput_0, blend=@MaterialExpressionMultiply_6) {
   code="max(base,blend)"
   description=Max
   inputs [
    {inputName=base, input {expression=&self:MaterialExpressionFunctionInput_0}}
    {inputName=blend, input {expression=&self:MaterialExpressionMultiply_6}}
   ]
  }
  MaterialExpressionMultiply_8 = Multiply(A=@n1, B=Constant() {r=2})
  n4 = Custom(base=@MaterialExpressionFunctionInput_0, blend=@MaterialExpressionMultiply_8) {
   code="min(base,blend)"
   description=Min
   inputs [
    {inputName=base, input {expression=&self:MaterialExpressionFunctionInput_0}}
    {inputName=blend, input {expression=&self:MaterialExpressionMultiply_8}}
   ]
  }
  n5 = If(A=ComponentMask(None=@n1) {r=true}, B=@n2, "A > B"=ComponentMask(None=@n3) {r=true}, "A < B"=ComponentMask(None=@n4) {r=true})
  n6 = If(A=ComponentMask(None=@n1) {g=true}, B=@n2, "A > B"=ComponentMask(None=@n3) {g=true}, "A < B"=ComponentMask(None=@n4) {g=true})
  n7 = ComponentMask(None=@n1) {b=true}
  n8 = ComponentMask(None=@n3) {b=true}
  n9 = ComponentMask(None=@n4) {b=true}
  FunctionOutput(None=AppendVector(A=AppendVector(A=@n5, B=@n6), B=If(A=@n7, B=@n2, "A > B"=@n8, "A < B"=@n9))) {description="Emulates Photoshop's Overlay blend.  The order of inputs does change the result."}
  n10 = Multiply(A=@MaterialExpressionFunctionInput_0, B=@MaterialExpressionMultiply_8)
  OneMinus(None=Multiply(B=OneMinus(None=@n10)))
 ]
}
