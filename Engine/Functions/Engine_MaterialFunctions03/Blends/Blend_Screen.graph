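# Engine material function that performs a Screen blend of two inputs, exposed to the material library under Misc and Blends.
# The Base and Blend inputs are each inverted with OneMinus, multiplied together, then inverted again for the single output.
# Each input uses a default engine texture sample's RGB as its preview value.
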MaterialFunction /Engine/Functions/Engine_MaterialFunctions03/Blends/Blend_Screen
graph {
 properties {
  bExposeToLibrary=true
  description="Inverts each texture, multiplies them together, then re-inverts the results."
  libraryCategoriesText [Misc,Blends]
 }
 nodes [
  n1 = OneMinus(None=FunctionInput(Preview=TextureSample() {texture=/Engine/EngineMaterials/BlendFunc_DefBase}.RGB) {description=Base texture input, inputName=Base})
  n2 = OneMinus(None=FunctionInput(Preview=TextureSample() {texture=/Engine/EngineMaterials/BlendFunc_DefBlend}.RGB) {description=Blend texture input, inputName=Blend})
  FunctionOutput(None=OneMinus(None=Multiply(A=@n1, B=@n2))) {description="Inverts each texture, multiplies them together, then re-inverts the results."}
 ]
}
